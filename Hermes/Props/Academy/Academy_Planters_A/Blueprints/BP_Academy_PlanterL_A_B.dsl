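; Academy large planter prop variant (A_B), a child Blueprint that adjusts the juniper hedge meshes in the planter.
; Its construction script runs the parent's construction, then overrides the world scale of two straight juniper hedge fallback components, narrowing them horizontally by different amounts while keeping full height.
; BeginPlay and Tick hold no active logic.
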(event EventBeginPlay
  (|CallParentFunction) ; This node is disabled and will not be called.\nDrag off pins to build functionality.
)

(event EventTick
  (params (DeltaSeconds :type Float%20%28single-precision%29)))

(fn ConstructionScript
  (|CallParentFunction)
  (Transformation|SetWorldScale3D (Variables|Default|GetSM_JuniperHedge_straight_fallback) "0.800000,0.800000,1.000000")
  (Transformation|SetWorldScale3D (Variables|Default|GetSM_JuniperHedge_straight_fallback1) "1.000000,0.800000,1.000000"))
