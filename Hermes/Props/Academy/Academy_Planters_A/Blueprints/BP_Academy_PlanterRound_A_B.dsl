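; Blueprint for a round academy planter prop, variant B, apparently with a juniper hedge cube fallback mesh component.
; Its construction script calls the parent script, then sets the world scale of that fallback component to unit scale.
; BeginPlay holds only a disabled parent call and Tick is empty, so it adds no runtime behavior of its own.
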(event EventBeginPlay
  (|CallParentFunction) ; This node is disabled and will not be called.\nDrag off pins to build functionality.
)

(event EventTick
  (params (DeltaSeconds :type Float%20%28single-precision%29)))

(fn ConstructionScript
  (|CallParentFunction)
  (Transformation|SetWorldScale3D (Variables|Default|GetSM_JuniperHedge_cube_fallback) "1.000000,1.000000,1.000000"))
